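# Sound cue for the reaction sound of the StrontiumSpark back bling cosmetic.
# A random node picks one of four equally weighted wave variations each time the cue plays.
# It uses close-range pet attenuation, the reactive back bling property sheet and the back bling sound class, with a reduced volume multiplier.
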SoundCue /BRCosmetics/Sounds/Backblings/StrontiumSpark/Backbling_StrontiumSpark_React_Cue
root=Random_0
nodes [
 {id=Random_0, children [WavePlayer_1,WavePlayer_2,WavePlayer_3,WavePlayer_4], properties {weights [1,1,1,1]}}
 {id=WavePlayer_1, properties {soundWaveAssetPtr=&/Backbling_StrontiumSpark_React_01}}
 {id=WavePlayer_2, properties {soundWaveAssetPtr=&/Backbling_StrontiumSpark_React_02}}
 {id=WavePlayer_3, properties {soundWaveAssetPtr=&/Backbling_StrontiumSpark_React_04}}
 {id=WavePlayer_4, properties {soundWaveAssetPtr=&/Backbling_StrontiumSpark_React_03}}
]
properties {
 attenuationSettings=&Fort_Pets_Close_Athena_Att
 audioPropertiesSheet=/BRCosmetics/Sounds/Settings/PropertySheets/Reactive/APS_Cosmetic_Reactive_Backbling
 maxDistance=1800.0001
 soundClassObject=/Game/Sounds/FortSoundClassesAndModes/Fort_Backblings
 volumeMultiplier=0.3
}
